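# NS_VentAir_Small is a Niagara system for a small air vent effect, judging by its name and folder.
# It combines three emitters, Air_Mesh, Air001 and Motes, whose names suggest mesh-based air, a second air layer and drifting motes.
# SystemState is the only listed module, with no user variables, data interfaces or dynamic inputs.
# No renderers or emitter internals are recorded here.
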NiagaraSystem /Game/Effects/Fort_Effects/Effects/AirventAir/NS_VentAir_Small
dependencies {usedDataInterfaces [], usedDynamicInputs [], usedModules [&SystemState]}
emitters {Air001 {rendererClasses [], renderers []}, Air_Mesh {rendererClasses [], renderers []}, Motes {rendererClasses [], renderers []}}
summary {
 systemName=NS_VentAir_Small
 emitters [{emitterName=Air_Mesh, rendererClasses []}, {emitterName=Air001, rendererClasses []}, {emitterName=Motes, rendererClasses []}]
 userVariables []
}
